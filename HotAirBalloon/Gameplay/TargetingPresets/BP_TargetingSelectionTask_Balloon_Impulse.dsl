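; Targeting selection task for the hot air balloon's impulse targeting, aligning the query with the source's direction of travel.
; It casts the targeting source to TrainCar and converts that actor's velocity to a yaw-only rotation, ignoring its actual facing.
; GetSourceRotation returns that rotation as a quaternion.
; GetSourceOffset applies the same yaw to an offset built from the ImpulseOffset X, Y and Z scalable floats.
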(fn GetSourceOffset
  (params (TargetingHandle :type Targeting%20Request%20Handle%20Structure%20%28by%20ref%29))
  (bind _astrain_car (Utilities|Casting|CastToTrainCar (Utilities|Struct|BreakTargetingSourceContext (TargetingSystem|TargetingTypes|GetTargetingSourceContext (Targeting|GetTargetingSubsystem TargetingHandle) TargetingHandle))))
  (return (Math|Vector|RotateVector (Math|Vector|MakeVector (ScalableFloat|GetValueatLevel (Variables|Default|GetImpulseOffsetX)) (ScalableFloat|GetValueatLevel (Variables|Default|GetImpulseOffsetY)) (ScalableFloat|GetValueatLevel (Variables|Default|GetImpulseOffsetZ))) (Math|Rotator|MakeRotator 0.0 0.0 (Fortnite|VectortoNormalizedAngleinDegrees (Transformation|GetVelocity _astrain_car))))))

(fn GetSourceRotation
  (params (TargetingHandle :type Targeting%20Request%20Handle%20Structure%20%28by%20ref%29))
  (bind _astrain_car (Utilities|Casting|CastToTrainCar (Utilities|Struct|BreakTargetingSourceContext (TargetingSystem|TargetingTypes|GetTargetingSourceContext (Targeting|GetTargetingSubsystem TargetingHandle) TargetingHandle))))
  (return (Math|Conversions|ToQuaternion(Rotator) (Math|Rotator|MakeRotator 0.0 0.0 (Fortnite|VectortoNormalizedAngleinDegrees (Transformation|GetVelocity _astrain_car))))))
